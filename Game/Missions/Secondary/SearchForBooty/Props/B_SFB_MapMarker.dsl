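; Map marker prop for the SearchForBooty secondary mission, showing a ground X decal and sky pillar indicator for a pirate chest.
; Construction and BeginPlay create dynamic materials, then a vertical trace orients the decal to the ground normal and yaws the indicator to match.
; Unhide and Hide events, called by the chest's passive ability, play one timeline forward with sound or in reverse to drive material and particle parameters.
; When the timeline finishes with ShouldHide set, the indicator is hidden.
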(event EventBeginPlay
  (bind (_returnvalue_roll _returnvalue_pitch) (Math|Rotator|BreakRotator (Transformation|GetWorldRotation (Variables|Default|GetP_Pirates_Chest_Indicator))))
  ;[1 1 1] Setup Decal Material Instance
  (bind _returnvalue (Rendering|Components|Decal|CreateDynamicMaterialInstance (Variables|Default|GetDecal)))
  (Variables|Default|SetMI_Decal _returnvalue)
  ;[1 1 1] FX - Nick P - Rotates decal to face surface normal of ground
  (bind (_outhit _returnvalue_2) (Collision|LineTraceByChannel (Math|Vector|MakeVector (.x (Transformation|GetWorldLocation (Variables|Default|GetDecal))) (.y (Transformation|GetWorldLocation (Variables|Default|GetDecal))) (+ (.z (Transformation|GetWorldLocation (Variables|Default|GetDecal))) (Variables|Default|GetDecalTraceRange))) (Math|Vector|MakeVector (.x (Transformation|GetWorldLocation (Variables|Default|GetDecal))) (.y (Transformation|GetWorldLocation (Variables|Default|GetDecal))) (- (.z (Transformation|GetWorldLocation (Variables|Default|GetDecal))) (Variables|Default|GetDecalTraceRange))) "TraceTypeQuery10"))
  (if _returnvalue_2
    (Transformation|SetWorldRotation (Variables|Default|GetDecal) (Utilities|Select (Math|Conversions|RotationFromXVector (.ImpactNormal (Collision|BreakHitResult _outhit))) (Math|Rotator|MakeRotator (.roll (Math|Conversions|RotationFromXVector (.ImpactNormal (Collision|BreakHitResult _outhit)))) -9e+01 (.yaw (Math|Conversions|RotationFromXVector (.ImpactNormal (Collision|BreakHitResult _outhit))))) (> (.pitch (Math|Conversions|RotationFromXVector (.ImpactNormal (Collision|BreakHitResult _outhit)))) -45.0)) false false)
    (bind (_returnvalue_roll_2 _returnvalue_pitch_2 _returnvalue_yaw) (Math|Rotator|BreakRotator (Math|Conversions|RotationFromXVector (.ImpactNormal (Collision|BreakHitResult _outhit)))))
    (Transformation|SetWorldRotation (Variables|Default|GetP_Pirates_Chest_Indicator) (Math|Rotator|MakeRotator _returnvalue_roll _returnvalue_pitch _returnvalue_yaw) false false)))

(event Custom|UnhideSkyMarker
  (bind _p_pirates_chest_indicator (Variables|Default|GetP_Pirates_Chest_Indicator))
  (bind _mid_glowy_x (Variables|Default|GetMID_Glowy_X))
  ;[1 1 1] Hide and Unhide then sky Marker. This is called from the passive ability that opperates the chest.
  (if (or (Variables|/Script/Engine.SceneComponent|GetbHiddenInGame (Variables|Default|GetP_Pirates_Chest_Indicator)) (Variables|Default|GetShouldHide))
    (Development|SetHiddeninGame (Variables|Default|GetDecal) false false)
    (Audio|PlaySoundatLocation "/Game/Sounds/Foley_Loot/Containers/Pirate_Booty/Pirate_Booty_Beacon_Activate_Cue.Pirate_Booty_Beacon_Activate_Cue" (Transformation|GetActorLocation))
    (Variables|Default|SetShouldHide false)
    (bind (_direction _opacity _spawnscale _spawnverticalscale _scale_x_mark_on_z _scale_x_mark_on_xy _scalecorona _pillarscale _secondaryopacity _opacity_x) (Timeline|SpawnMarkerAnim PlayFromStart)
      (:Update
        (Rendering|Material|SetScalarParameterValue _mid_glowy_x "Scale_Z" _scale_x_mark_on_z :!class /Script/Engine.MaterialInstanceDynamic)
        (Rendering|Material|SetScalarParameterValue _mid_glowy_x "Scale_XY" _scale_x_mark_on_xy :!class /Script/Engine.MaterialInstanceDynamic)
        (Effects|Components|ParticleSystem|SetVectorParameter _p_pirates_chest_indicator "Corona Scale Mult" (Math|Vector|MakeVector _scalecorona _scalecorona _scalecorona))
        (Rendering|Material|SetScalarParameterValue (Variables|Default|GetMID_Glow_Pillar_Mesh) "MaxScaleIncrease" _pillarscale :!class /Script/Engine.MaterialInstanceDynamic)
        (Effects|Components|ParticleSystem|SetFloatParameter _p_pirates_chest_indicator "SecondaryOpacity" _secondaryopacity)
        (Rendering|Material|SetScalarParameterValue (Variables|Default|GetMI_Decal) "X Decal Fire Lerp" _opacity_x :!class /Script/Engine.MaterialInstanceDynamic))
      (:Finished
        (if (Variables|Default|GetShouldHide)
          (Development|SetHiddeninGame (Variables|Default|GetP_Pirates_Chest_Indicator) true false))))))

(event Custom|HideSkyMarker
  (bind _p_pirates_chest_indicator (Variables|Default|GetP_Pirates_Chest_Indicator))
  (bind _mid_glowy_x (Variables|Default|GetMID_Glowy_X))
  ;[1 1 1] Hide and Unhide then sky Marker. This is called from the passive ability that opperates the chest.
  (Variables|Default|SetShouldHide true)
  (Timeline|SpawnMarkerAnim ReverseFromEnd))

(fn ConstructionScript
  (bind _returnvalue (Rendering|Material|CreateNamedDynamicMaterialInstance (Variables|Default|GetP_Pirates_Chest_Indicator) "X_Mesh" "/Game/Effects/Fort_Effects/Gameplay/PirateChests/MI_X_Ground_Decal_Glow.MI_X_Ground_Decal_Glow"))
  (Variables|Default|SetMID_Glowy_X _returnvalue)
  (bind _returnvalue_2 (Rendering|Material|CreateNamedDynamicMaterialInstance (Variables|Default|GetP_Pirates_Chest_Indicator) "GlowPillarMesh" "/Game/Effects/Fort_Effects/Gameplay/PirateChests/MI_Booty_Indicator_Pillar_Glow.MI_Booty_Indicator_Pillar_Glow"))
  (Variables|Default|SetMID_Glow_Pillar_Mesh _returnvalue_2))
